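; Parent Blueprint for wet cement car props, built on a static mesh actor with a selectable train color.
; The construction script clamps the TrainColor number to a range of 0 to 3 and writes it to custom primitive data index 3 on the mesh component.
; This gives each instance its own value for a material to read, without creating dynamic material instances.
; BeginPlay, overlap and Tick events are empty, so it has no runtime logic.
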(event EventBeginPlay)

(event Collision|EventActorBeginOverlap
  (params (OtherActor :type Actor%20Object%20Reference)))

(event EventTick
  (params (DeltaSeconds :type Float%20%28single-precision%29)))

(fn ConstructionScript
  (bind _traincolor (Variables|TrainColor|GetTrainColor%20#))
  (if (< _traincolor 0)
    (Variables|TrainColor|SetTrainColor%20# 0.0)
    (elif (> _traincolor 3.0)
      (Variables|TrainColor|SetTrainColor%20# 3.0)
      (else
        (Variables|TrainColor|SetTrainColor%20# (Variables|TrainColor|GetTrainColor%20#)))))
  (Rendering|Material|SetCustomPrimitiveDataFloat (Variables|StaticMeshActor|GetStaticMeshComponent) 3 (Variables|TrainColor|GetTrainColor%20#)))
