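# M_DecalMasked_A is a masked, Default Lit material for decal-style detail on props, defaulting to a train car decal texture.
# The Decal texture parameter's RGB is scaled by BaseColorMult for color, while its alpha is boosted by OpacityMult and saturated into a hardened cutout mask.
# Scalar parameters expose Metallic, Specular and Roughness.
# It uses the surface domain rather than deferred decal, and automatic usage flagging in the editor is off.
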Material /Hermes/Props/WetCement/Materials/M_DecalMasked_A
graph {
 properties {bAutomaticallySetUsageInEditor=false, blendMode=BLEND_Masked, usedShadingModels=Default Lit}
 nodes [
  n1 = TextureSampleParameter2D() {parameterName=Decal, texture=/Hermes/Props/WetCement/Textures/T_Hermes_Train_Car_Decal_A_D}
  Multiply(A=@n1.RGB, B=ScalarParameter() {defaultValue=1.5, parameterName=BaseColorMult})
  Saturate(None=Multiply(A=@n1.A, B=ScalarParameter() {defaultValue=10, parameterName=OpacityMult}))
  ScalarParameter() {parameterName=Metallic}
  ScalarParameter() {defaultValue=1, parameterName=Specular}
  ScalarParameter() {defaultValue=0.5, parameterName=Roughness}
 ]
}
